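# Translucent unlit sprite material for Cascade and Niagara particles that creates an explosion distortion wave, using index-of-refraction mode.
# A pointy normal texture blends toward 1 outside a soft flare mask scaled by particle alpha, then gets clamped, shaping the distortion.
# A scalar blends in particle color through the shared emissive adjustments function.
# The flare mask also forms a separate opacity-style term, with a static switch for depth fade.
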Material /Game/Effects/Fort_Effects/Effects/Explosion/M_Distortion_Wave
graph {
 properties {
  bGenerateSphericalParticleNormals=true
  bUsedWithNiagaraSprites=true
  bUsedWithParticleSprites=true
  blendMode=BLEND_Translucent
  pixelDepthOffsetMode=PDOM_Legacy
  refractionMethod=RM_IndexOfRefraction
  shadingModel=MSM_Unlit
  usedShadingModels=Unlit
 }
 nodes [
  n1 = ParticleColor()
  n2 = TextureSample() {texture=/Game/Global/Textures/Default/Masks_and_Gradients/T_New_Soft_Flair}
  n3 = Multiply(A=@n1.A, B=@n2.G)
  Clamp(None=LinearInterpolate(
    A=TextureSample() {samplerType=SAMPLERTYPE_Normal, texture=/Game/Effects/Fort_Effects/Textures/PointyNormal}.RGB
    B=Constant() {r=1}
    Alpha=OneMinus(None=@n3)
   ))
  MaterialFunctionCall(Emissive=LinearInterpolate(B=@n1.RGB, Alpha=ScalarParameter() {parameterName=UseParticleColor})) {materialFunction=&MF_FX_EmissiveAdjustments}
  n4 = ScalarParameter() {parameterName=FadeDistance}
  n5 = Multiply(A=Constant() {r=1}, B=@n2.G)
  StaticSwitchParameter(True=Multiply(A=DepthFade(FadeDistance=@n4), B=@n5), False=@n5) {parameterName=UseDepthFade?}
 ]
}
