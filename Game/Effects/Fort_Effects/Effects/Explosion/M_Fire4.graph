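# Masked, lit mesh-particle material for stylized explosion fire and smoke.
# Blends a static normal map with time-panned, world-aligned cloud normals, then softens them toward the vertex normal for fake lighting.
# A sharpened Fresnel-plus-up-facing mask blends particle color into day/night-adjusted bloom over lit smoke tint, with dynamic parameters controlling edge, glow, panning and scale.
# Particle alpha drives a HeightLerp dissolve using camera-facing normals as height.
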Material /Game/Effects/Fort_Effects/Effects/Explosion/M_Fire4
graph {
 properties {
  bGenerateSphericalParticleNormals=true
  bTangentSpaceNormal=false
  bUsedWithMeshParticles=true
  blendMode=BLEND_Masked
  pixelDepthOffsetMode=PDOM_Legacy
  usedShadingModels=Default Lit
 }
 nodes [
  n1 = TextureSample() {samplerType=SAMPLERTYPE_Normal, texture=/Game/Effects/Fort_Effects/Textures/Explo/T_ExploTest_02_N}
  n2 = VectorParameter() {parameterName=ScaleStaticNormal, defaultValue=<rgba 1 1 1 0>}
  n3 = VectorParameter() {parameterName=ScaleWorldNormal, defaultValue=<rgba 1 1 1 0>}
  n4 = DynamicParameter() {defaultValue=<rgba 0 0 0 1>, paramNames [Texture Size, ControlShadowEdge, PanTexture, ControlBloom]}
  n5 = Time()
  n6 = Multiply(A=@n4.PanTexture, B=Constant3Vector() {constant=<rgba 0 0 1 0>})
  n7 = MaterialFunctionCall!(
   Diffuse Texture Object=TextureObject() {samplerType=SAMPLERTYPE_LinearColor, texture=/Game/Global/Textures/Default/Tileable_Patterns/Cloud_Diffuse}
   Spec Texture Object
   TextureSize=Multiply(A=Constant3Vector() {constant=<rgba 1 1 1 0>}, B=@n4.Texture Size) {constB=3}
   --------------------------
   Use seperate settings for Z projection
   Use seperate texture for Z projection
   Texture Size Z projection
   Z Plane Diffuse Texture Object
   Z Plane Spec Texture Object
   "--------------------------      "
   ProjectionTransitionContrast
   Use Texture Alpha Heightmap to Refine the Transistion
   HeightLerpTransitionContrast
   PickTheDiffuseHeightChannel
   TheDiffuseHeightChannel
   "--------------------------       "
   NormalTextureObject=TextureObject() {samplerType=SAMPLERTYPE_Normal, texture=/Game/Global/Textures/Default/Tileable_Patterns/Cloud_normals}
   World Space Normals
   "--------------------------  "
   Export Float 4 Diffuse
   Export Float 4 Spec
   "--------------------------     "
   WorldPosition=Add(A=Multiply(A=@n5, B=@n6) {constB=200}, B=WorldPosition().XYZ)
   World Space Normal
  ) {materialFunction=/Engine/Functions/Engine_MaterialFunctions02/Texturing/WorldAlignedTextures_Complex}
  n8 = LinearInterpolate(
   A=Normalize(VectorInput=Multiply(A=@n1.RGB, B=@n2.RGB))
   B=Normalize(VectorInput=Multiply(A=@n3.RGB, B=@n7.Normal XYZ Texture))
   Alpha=ScalarParameter() {defaultValue=0.5, parameterName=LerpStaticNormal}
  )
  n9 = LinearInterpolate(A=@n8, B=VertexNormalWS(), Alpha=ScalarParameter() {defaultValue=0.2, parameterName=FlattenNormal}) {desc=Soft normal}
  n10 = Constant3Vector() {desc=Angle of light, constant=<rgba 0.2 1.2 0 0>}
  n11 = Add(A=DotProduct(A=@n9, B=Transform(None=@n10) {transformSourceType=TRANSFORMSOURCE_View})) {constB=-0.1}
  n12 = Multiply(
   A=Clamp(None=Multiply(A=@n11) {constB=40})
   B=VectorParameter() {parameterName=Color-BaseSmoke, defaultValue=<rgba 0.1327 0.1174 0.1083 1>}.RGB
  ) {constB=0.5}
  n13 = ParticleColor()
  n14 = ParticleColor()
  n15 = Constant3Vector() {constant=<rgba 4 0.7023 0.2042 0>}
  n16 = Multiply(
   A=MaterialFunctionCall!(DayBloom_Mult=Constant() {r=1}, In=StaticSwitchParameter(True=@n14.RGB, False=@n15) {parameterName=ParticleColorControl}) {materialFunction=/Game/Packages/Fortnite_UniqueMaterials_1/MH_Materials/DayNight_Auto_Bloom_Control}.Result
   B=DynamicParameter() {defaultValue=<rgba 0 0 0 1>, paramNames [Texture Size, ControlShadowEdge, PanTexture, ControlBloom]}.ControlBloom
  )
  n17 = Add(A=Multiply(A=Fresnel()) {constB=10}, B=DotProduct(A=@n9, B=Constant3Vector() {constant=<rgba 0 0 1 0>}))
  n18 = DynamicParameter() {defaultValue=<rgba 0 0 0 1>, paramNames [Texture Size, ControlShadowEdge, PanTexture, ControlBloom]}
  n19 = Clamp(None=Multiply(A=Add(A=@n17, B=@n18.ControlShadowEdge) {constB=-0.4, desc="1-0 (line)"}) {constB=15, desc=sharp line})
  MaterialFunctionCall(Emissive=Add(A=Multiply(A=@n12) {constB=0.2}, B=LinearInterpolate(A=@n13.RGB, B=@n16, Alpha=@n19))) {materialFunction=&MF_FX_EmissiveAdjustments}
  MaterialFunctionCall(Normal=@n8, Flatness=ScalarParameter() {defaultValue=4, parameterName=FlattenLitNormal}) {materialFunction=&FlattenNormal}
  n20 = Clamp(None=DotProduct(A=@n9, B=CameraVectorWS()))
  n21 = Clamp(None=DotProduct(A=VertexNormalWS(), B=CameraVectorWS()))
  MaterialFunctionCall(Transition Phase=@n13.A, Height Texture=Divide(A=Add(A=@n20, B=@n21))) {materialFunction=&HeightLerp}
  ScalarParameter() {defaultValue=1, parameterName=Roughness}
  ScalarParameter() {defaultValue=0.5, parameterName=Metallic}
  VectorParameter() {parameterName=Color-Base, defaultValue=<rgba 0 0 0 1>}
 ]
}
